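# Physics asset for the cape and pack variant of the ChiliAbide male character, defining bodies for cape simulation.
# Center, left and right capsule chains hang from MSTR bodies, with swing limits widening down each chain.
# Cross constraints with free swing and twist tie the side chains to the center from segment 04 down, doubled in reverse at 06 and 07.
# Torso, clavicle, knife, weapon and ground shapes add collision, while limb bodies are empty.
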PhysicsAsset /BRCosmetics/Characters/Player/Male/Medium/Bodies/M_MED_ChiliAbide/Meshes/M_MED_ChiliAbide_New_Cape_Pack_physics1
bodies {
 calf_l []
 calf_r []
 clavicle_l [{shapeType=Capsule, length=9.0192, radius=9.8871, center=<9.4143 -0.5389 -2.5177>, rotation=<rot 69.3297 1.4476 176.9462>}]
 clavicle_r [{shapeType=Capsule, length=9.0192, radius=9.8871, center=<-9.4143 0.5389 2.5177>, rotation=<rot 69.3297 1.4476 -3.0538>}]
 dyn_cape_01 [{shapeType=Capsule, length=11.7556, radius=5.9773, shapeName=dyn_cape_01_capsule, center=<7.15 0 11.8488>, rotation=<rot 90 0 0>}]
 dyn_cape_02 [{shapeType=Capsule, length=10, radius=10, shapeName=dyn_cape_01_capsule, center=<7.15 0 9.5994>, rotation=<rot 90 0 0>}]
 dyn_cape_03 [{shapeType=Capsule, length=10, radius=12, shapeName=dyn_cape_01_capsule, center=<7.15 0 10.6212>, rotation=<rot 90 0 0>}]
 dyn_cape_04 [{shapeType=Capsule, length=10, radius=12, shapeName=dyn_cape_01_capsule, center=<7.15 0 12.3005>, rotation=<rot 90 0 0>}]
 dyn_cape_05 [{shapeType=Capsule, length=10, radius=12, shapeName=dyn_cape_01_capsule, center=<7.15 0 12.3985>, rotation=<rot 90 0 0>}]
 dyn_cape_06 [{shapeType=Capsule, length=10, radius=12, shapeName=dyn_cape_01_capsule, center=<7.15 0 12.8561>, rotation=<rot 90 0 0>}]
 dyn_cape_07 [{shapeType=Capsule, length=10, radius=14, shapeName=dyn_cape_01_capsule, center=<5.8445 0 12.8561>, rotation=<rot 90 0 0>}]
 dyn_cape_MSTR [{shapeType=Capsule, length=0.505, radius=0.505, shapeName=dyn_cape_r_MSTR_capsule, rotation=<rot 0 -90 0>}]
 dyn_cape_l_01 [{shapeType=Capsule, length=11.7556, radius=5.9773, shapeName=dyn_cape_01_capsule, center=<7.15 0 10.6491>, rotation=<rot 90 0 0>}]
 dyn_cape_l_02 [{shapeType=Capsule, length=10, radius=10, shapeName=dyn_cape_01_capsule, center=<7.15 4 9.6122>, rotation=<rot 90 0 0>}]
 dyn_cape_l_03 [{shapeType=Capsule, length=10, radius=12, shapeName=dyn_cape_01_capsule, center=<7.15 7 9.3861>, rotation=<rot 90 0 0>}]
 dyn_cape_l_04 [{shapeType=Capsule, length=10, radius=12, shapeName=dyn_cape_01_capsule, center=<7.15 7 10.1258>, rotation=<rot 90 0 0>}]
 dyn_cape_l_05 [{shapeType=Capsule, length=10, radius=12, shapeName=dyn_cape_01_capsule, center=<7.15 7 11.1603>, rotation=<rot 90 0 0>}]
 dyn_cape_l_06 [{shapeType=Capsule, length=10, radius=12, shapeName=dyn_cape_01_capsule, center=<7.15 7 11.1603>, rotation=<rot 90 0 0>}]
 dyn_cape_l_07 [{shapeType=Capsule, length=10, radius=14, shapeName=dyn_cape_01_capsule, center=<5.8445 10 11.1603>, rotation=<rot 90 0 0>}]
 dyn_cape_l_MSTR [{shapeType=Capsule, length=0.505, radius=0.505, shapeName=dyn_cape_r_MSTR_capsule, rotation=<rot 0 -90 0>}]
 dyn_cape_r_01 [{shapeType=Capsule, length=11.7556, radius=5.9773, shapeName=dyn_cape_01_capsule, center=<7.15 0 10.6491>, rotation=<rot 90 0 0>}]
 dyn_cape_r_02 [{shapeType=Capsule, length=10, radius=10, shapeName=dyn_cape_01_capsule, center=<7.15 -4 9.3124>, rotation=<rot 90 0 0>}]
 dyn_cape_r_03 [{shapeType=Capsule, length=10, radius=12, shapeName=dyn_cape_01_capsule, center=<7.15 -7 9.3861>, rotation=<rot 90 0 0>}]
 dyn_cape_r_04 [{shapeType=Capsule, length=10, radius=12, shapeName=dyn_cape_01_capsule, center=<7.15 -7 10.1258>, rotation=<rot 90 0 0>}]
 dyn_cape_r_05 [{shapeType=Capsule, length=10, radius=12, shapeName=dyn_cape_01_capsule, center=<7.15 -7 11.1603>, rotation=<rot 90 0 0>}]
 dyn_cape_r_06 [{shapeType=Capsule, length=10, radius=12, shapeName=dyn_cape_01_capsule, center=<7.15 -7 11.1603>, rotation=<rot 90 0 0>}]
 dyn_cape_r_07 [{shapeType=Capsule, length=10, radius=14, shapeName=dyn_cape_01_capsule, center=<5.8445 -10 11.1603>, rotation=<rot 90 0 0>}]
 dyn_cape_r_MSTR [{shapeType=Capsule, length=0.505, radius=0.505, shapeName=dyn_cape_r_MSTR_capsule, rotation=<rot 0 -90 0>}]
 dyn_knife [{
   shapeType=Capsule
   length=21.5462
   radius=5.207
   shapeName=dyn_knife_capsule
   center=<1.1953 -3.0361 -2.0204>
   rotation=<rot 0 -96.6804 -64.3383>
  }]
 dyn_weapon_r [
  {
   shapeType=Capsule
   length=25.4136
   radius=8.9412
   shapeName=dyn_weapon_r_capsule
   center=<2.606 -0.5066 -3.5956>
   rotation=<rot 0 96.1318 -88.8083>
  }
  {
   shapeType=Capsule
   length=8.4451
   radius=8.9412
   shapeName=dyn_weapon_r_capsule
   center=<-10.5321 -2.7404 -4.0791>
   rotation=<rot -1.0746 160.5162 -89.4848>
  }
 ]
 ground_collision [{shapeType=Box, extentX=200, extentY=200, extentZ=12.412, shapeName=root_box, center=<0 0 -4.8416>}]
 lowerarm_l []
 lowerarm_r []
 pelvis [{shapeType=Capsule, length=9.6042, radius=11.1441, center=<-2.4471 -5.2421 -0.1981>}]
 spine_01 [{shapeType=Capsule, length=12.3379, radius=16.8776, center=<-1.7965 3.1242 0>}]
 spine_02 [
  {shapeType=Capsule, length=9.8884, radius=15.9303, center=<-1.711 1.7727 0.0001>}
  {shapeType=Capsule, length=29.875, radius=11.1013, center=<2.4051 -1.4367 0>, rotation=<rot 90 -12.2939 0>}
 ]
 spine_03 [
  {shapeType=Capsule, length=13.3712, radius=14.7147, center=<0.1445 -1.8858 0.0001>}
  {
   shapeType=Capsule
   length=15
   radius=15
   shapeName=spine_03_capsule
   center=<0.0236 0.0975 -0.0016>
   rotation=<rot -86.1941 179.9999 172.7136>
  }
 ]
 spine_05 [{shapeType=Capsule, length=10.3836, radius=15.695, center=<-2.2795 0.501 0>}]
 thigh_l []
 thigh_r []
 upperarm_l []
 upperarm_r []
}
constraints [
 {bone1Name=dyn_cape_01, bone2Name=dyn_cape_MSTR, swing1LimitDegrees=10, swing2LimitDegrees=0, twistLimitDegrees=5}
 {bone1Name=dyn_cape_02, bone2Name=dyn_cape_01, swing1LimitDegrees=20, swing2LimitDegrees=0, twistLimitDegrees=5}
 {bone1Name=dyn_cape_03, bone2Name=dyn_cape_02, swing1LimitDegrees=45, swing2LimitDegrees=5, twistLimitDegrees=5}
 {bone1Name=dyn_cape_04, bone2Name=dyn_cape_03, swing1LimitDegrees=45, swing2LimitDegrees=45, twistLimitDegrees=5}
 {
  bone1Name=dyn_cape_r_04
  bone2Name=dyn_cape_04
  swing1LimitDegrees=45
  swing1Motion=Free
  swing2LimitDegrees=45
  swing2Motion=Free
  twistLimitDegrees=45
  twistMotion=Free
 }
 {
  bone1Name=dyn_cape_l_04
  bone2Name=dyn_cape_04
  swing1LimitDegrees=45
  swing1Motion=Free
  swing2LimitDegrees=45
  swing2Motion=Free
  twistLimitDegrees=45
  twistMotion=Free
 }
 {bone1Name=dyn_cape_05, bone2Name=dyn_cape_04, swing1LimitDegrees=45, swing2LimitDegrees=45, twistLimitDegrees=5}
 {
  bone1Name=dyn_cape_r_05
  bone2Name=dyn_cape_05
  swing1LimitDegrees=45
  swing1Motion=Free
  swing2LimitDegrees=45
  swing2Motion=Free
  twistLimitDegrees=45
  twistMotion=Free
 }
 {
  bone1Name=dyn_cape_l_05
  bone2Name=dyn_cape_05
  swing1LimitDegrees=45
  swing1Motion=Free
  swing2LimitDegrees=45
  swing2Motion=Free
  twistLimitDegrees=45
  twistMotion=Free
 }
 {bone1Name=dyn_cape_06, bone2Name=dyn_cape_05, swing1LimitDegrees=45, swing2LimitDegrees=45, twistLimitDegrees=5}
 {
  bone1Name=dyn_cape_r_06
  bone2Name=dyn_cape_06
  swing1LimitDegrees=45
  swing1Motion=Free
  swing2LimitDegrees=45
  swing2Motion=Free
  twistLimitDegrees=45
  twistMotion=Free
 }
 {
  bone1Name=dyn_cape_l_06
  bone2Name=dyn_cape_06
  swing1LimitDegrees=45
  swing1Motion=Free
  swing2LimitDegrees=45
  swing2Motion=Free
  twistLimitDegrees=45
  twistMotion=Free
 }
 {
  bone1Name=dyn_cape_06
  bone2Name=dyn_cape_r_06
  swing1LimitDegrees=45
  swing1Motion=Free
  swing2LimitDegrees=45
  swing2Motion=Free
  twistLimitDegrees=45
  twistMotion=Free
 }
 {
  bone1Name=dyn_cape_06
  bone2Name=dyn_cape_l_06
  swing1LimitDegrees=45
  swing1Motion=Free
  swing2LimitDegrees=45
  swing2Motion=Free
  twistLimitDegrees=45
  twistMotion=Free
 }
 {bone1Name=dyn_cape_07, bone2Name=dyn_cape_06, swing1LimitDegrees=45, swing2LimitDegrees=45, twistLimitDegrees=5}
 {
  bone1Name=dyn_cape_l_07
  bone2Name=dyn_cape_07
  swing1LimitDegrees=45
  swing1Motion=Free
  swing2LimitDegrees=45
  swing2Motion=Free
  twistLimitDegrees=45
  twistMotion=Free
 }
 {
  bone1Name=dyn_cape_r_07
  bone2Name=dyn_cape_07
  swing1LimitDegrees=45
  swing1Motion=Free
  swing2LimitDegrees=45
  swing2Motion=Free
  twistLimitDegrees=45
  twistMotion=Free
 }
 {
  bone1Name=dyn_cape_07
  bone2Name=dyn_cape_r_07
  swing1LimitDegrees=45
  swing1Motion=Free
  swing2LimitDegrees=45
  swing2Motion=Free
  twistLimitDegrees=45
  twistMotion=Free
 }
 {
  bone1Name=dyn_cape_07
  bone2Name=dyn_cape_l_07
  swing1LimitDegrees=45
  swing1Motion=Free
  swing2LimitDegrees=45
  swing2Motion=Free
  twistLimitDegrees=45
  twistMotion=Free
 }
 {bone1Name=dyn_cape_l_01, bone2Name=dyn_cape_l_MSTR, swing1LimitDegrees=10, swing2LimitDegrees=0, twistLimitDegrees=5}
 {bone1Name=dyn_cape_l_02, bone2Name=dyn_cape_l_01, swing1LimitDegrees=20, swing2LimitDegrees=0, twistLimitDegrees=5}
 {bone1Name=dyn_cape_l_03, bone2Name=dyn_cape_l_02, swing1LimitDegrees=45, swing2LimitDegrees=5, twistLimitDegrees=5}
 {bone1Name=dyn_cape_l_04, bone2Name=dyn_cape_l_03, swing1LimitDegrees=45, swing2LimitDegrees=45, twistLimitDegrees=5}
 {bone1Name=dyn_cape_l_05, bone2Name=dyn_cape_l_04, swing1LimitDegrees=45, swing2LimitDegrees=45, twistLimitDegrees=5}
 {bone1Name=dyn_cape_l_06, bone2Name=dyn_cape_l_05, swing1LimitDegrees=45, swing2LimitDegrees=45, twistLimitDegrees=5}
 {bone1Name=dyn_cape_l_07, bone2Name=dyn_cape_l_06, swing1LimitDegrees=45, swing2LimitDegrees=45, twistLimitDegrees=5}
 {bone1Name=dyn_cape_r_01, bone2Name=dyn_cape_r_MSTR, swing1LimitDegrees=10, swing2LimitDegrees=0, twistLimitDegrees=5}
 {bone1Name=dyn_cape_r_02, bone2Name=dyn_cape_r_01, swing1LimitDegrees=20, swing2LimitDegrees=0, twistLimitDegrees=5}
 {bone1Name=dyn_cape_r_03, bone2Name=dyn_cape_r_02, swing1LimitDegrees=45, swing2LimitDegrees=5, twistLimitDegrees=5}
 {bone1Name=dyn_cape_r_04, bone2Name=dyn_cape_r_03, swing1LimitDegrees=45, swing2LimitDegrees=45, twistLimitDegrees=5}
 {bone1Name=dyn_cape_r_05, bone2Name=dyn_cape_r_04, swing1LimitDegrees=45, swing2LimitDegrees=45, twistLimitDegrees=5}
 {bone1Name=dyn_cape_r_06, bone2Name=dyn_cape_r_05, swing1LimitDegrees=45, swing2LimitDegrees=45, twistLimitDegrees=5}
 {bone1Name=dyn_cape_r_07, bone2Name=dyn_cape_r_06, swing1LimitDegrees=45, swing2LimitDegrees=45, twistLimitDegrees=5}
]
