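# Physics asset supplying simplified collision for the cape of the male medium ChiliAbide player body, as its name indicates.
# Capsules cover the pelvis, five spine bones, clavicles and upper neck, with doubled capsules on pelvis and spine_02 and a sphere on the head.
# Arm, leg, foot, knife and weapon bodies carry no shapes, and there are no constraints, so it describes torso and head colliders rather than a ragdoll.
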PhysicsAsset /BRCosmetics/Characters/Player/Male/Medium/Bodies/M_MED_ChiliAbide/Meshes/M_MED_ChiliAbide_CapeColliders
bodies {
 calf_l []
 calf_r []
 clavicle_l [{shapeType=Capsule, length=9.0192, radius=9.8871, center=<9.4143 -0.5389 -2.5177>, rotation=<rot 69.3297 1.4476 176.9462>}]
 clavicle_r [{shapeType=Capsule, length=9.0192, radius=9.8871, center=<-9.4143 0.5389 2.5177>, rotation=<rot 69.3297 1.4476 -3.0538>}]
 dyn_knife []
 dyn_weapon_r []
 foot_l []
 foot_r []
 head [{shapeType=Sphere, radius=17.3125, center=<11.028 -0.174 0>}]
 lowerarm_l []
 lowerarm_r []
 neck_02 [{shapeType=Capsule, length=13.4375, radius=7.3175, center=<-0.2642 -0.3408 0>, rotation=<rot 90 -171.1193 -175.3978>}]
 pelvis [
  {shapeType=Capsule, length=8.6456, radius=15.1445, center=<4.464 -0.9203 7.615>, rotation=<rot 7.9422 179.9999 180>}
  {shapeType=Capsule, length=15, radius=15, shapeName=pelvis_capsule}
 ]
 spine_01 [{shapeType=Capsule, length=12.3379, radius=16.8776, center=<-1.7965 3.1242 0>}]
 spine_02 [
  {shapeType=Capsule, length=9.8884, radius=15.9303, center=<-1.711 3.7106 0.0001>}
  {shapeType=Capsule, length=29.875, radius=11.1013, center=<2.4704 0 0>, rotation=<rot 90 -0.0396 -0.0396>}
 ]
 spine_03 [{shapeType=Capsule, length=13.3712, radius=14.7147, center=<0.1445 1.5743 0.0001>}]
 spine_04 [{shapeType=Capsule, length=11.9285, radius=14.647, center=<2.766 -0.4252 0.0002>}]
 spine_05 [{shapeType=Capsule, length=10.3836, radius=15.695, center=<-2.2795 0.501 0>}]
 thigh_l []
 thigh_r []
 upperarm_l []
 upperarm_r []
}
constraints []
