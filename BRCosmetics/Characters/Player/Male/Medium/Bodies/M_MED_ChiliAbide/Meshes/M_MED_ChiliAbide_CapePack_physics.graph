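# Physics asset for the ChiliAbide male medium character's cape pack, providing bodies and constraints for cape dynamics.
# Three capsule chains (center, left, right) hang from spine_05, with tight twist and swing limits that widen down the chain.
# Free-rotation cross constraints tie the side chains to the center from segment 04 down, doubled in reverse on the last two rows.
# Torso, head and clavicle capsules plus a root floor box supply collision shapes; limb bodies are empty.
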PhysicsAsset /BRCosmetics/Characters/Player/Male/Medium/Bodies/M_MED_ChiliAbide/Meshes/M_MED_ChiliAbide_CapePack_physics
bodies {
 calf_l []
 calf_r []
 clavicle_l [{shapeType=Capsule, length=6.5623, radius=7.5194, center=<3.9566 -1.0862 -2.1127>, rotation=<rot -77.5566 177.3915 7.7314>}]
 clavicle_r [{shapeType=Capsule, length=6.5623, radius=7.5194, center=<-3.9566 1.0862 2.1127>, rotation=<rot -77.5566 177.3915 -172.2686>}]
 dyn_cape_02 [{shapeType=Capsule, length=10, radius=10, shapeName=dyn_cape_01_capsule, center=<7.15 0 5>, rotation=<rot 90 0 0>}]
 dyn_cape_03 [{shapeType=Capsule, length=10, radius=12, shapeName=dyn_cape_01_capsule, center=<7.15 0 5>, rotation=<rot 90 0 0>}]
 dyn_cape_04 [{shapeType=Capsule, length=10, radius=12, shapeName=dyn_cape_01_capsule, center=<7.15 0 5>, rotation=<rot 90 0 0>}]
 dyn_cape_05 [{shapeType=Capsule, length=10, radius=12, shapeName=dyn_cape_01_capsule, center=<7.15 0 5>, rotation=<rot 90 0 0>}]
 dyn_cape_06 [{shapeType=Capsule, length=10, radius=12, shapeName=dyn_cape_01_capsule, center=<7.15 0 5>, rotation=<rot 90 0 0>}]
 dyn_cape_07 [{shapeType=Capsule, length=10, radius=14, shapeName=dyn_cape_01_capsule, center=<5.8445 0 5>, rotation=<rot 90 0 0>}]
 dyn_cape_l_02 [{shapeType=Capsule, length=10, radius=10, shapeName=dyn_cape_01_capsule, center=<7.15 4 5>, rotation=<rot 90 0 0>}]
 dyn_cape_l_03 [{shapeType=Capsule, length=10, radius=12, shapeName=dyn_cape_01_capsule, center=<7.15 7 5>, rotation=<rot 90 0 0>}]
 dyn_cape_l_04 [{shapeType=Capsule, length=10, radius=12, shapeName=dyn_cape_01_capsule, center=<7.15 7 5>, rotation=<rot 90 0 0>}]
 dyn_cape_l_05 [{shapeType=Capsule, length=10, radius=12, shapeName=dyn_cape_01_capsule, center=<7.15 7 5>, rotation=<rot 90 0 0>}]
 dyn_cape_l_06 [{shapeType=Capsule, length=10, radius=12, shapeName=dyn_cape_01_capsule, center=<7.15 7 5>, rotation=<rot 90 0 0>}]
 dyn_cape_l_07 [{shapeType=Capsule, length=10, radius=14, shapeName=dyn_cape_01_capsule, center=<5.8445 10 5>, rotation=<rot 90 0 0>}]
 dyn_cape_r_02 [{shapeType=Capsule, length=10, radius=10, shapeName=dyn_cape_01_capsule, center=<7.15 -4 5>, rotation=<rot 90 0 0>}]
 dyn_cape_r_03 [{shapeType=Capsule, length=10, radius=12, shapeName=dyn_cape_01_capsule, center=<7.15 -7 5>, rotation=<rot 90 0 0>}]
 dyn_cape_r_04 [{shapeType=Capsule, length=10, radius=12, shapeName=dyn_cape_01_capsule, center=<7.15 -7 5>, rotation=<rot 90 0 0>}]
 dyn_cape_r_05 [{shapeType=Capsule, length=10, radius=12, shapeName=dyn_cape_01_capsule, center=<7.15 -7 5>, rotation=<rot 90 0 0>}]
 dyn_cape_r_06 [{shapeType=Capsule, length=10, radius=12, shapeName=dyn_cape_01_capsule, center=<7.15 -7 5>, rotation=<rot 90 0 0>}]
 dyn_cape_r_07 [{shapeType=Capsule, length=10, radius=14, shapeName=dyn_cape_01_capsule, center=<5.8445 -10 5>, rotation=<rot 90 0 0>}]
 head [{shapeType=Capsule, length=0, radius=18, center=<9.0698 0.0746 0>, rotation=<rot 0 -90 0>}]
 lowerarm_l []
 lowerarm_r []
 neck_02 []
 pelvis [{shapeType=Capsule, length=9.6042, radius=11.1441, center=<-2.4471 -2.5289 -0.1981>}]
 root [{shapeType=Box, extentX=200, extentY=200, extentZ=12.412, shapeName=root_box, center=<0 0 -4.8416>}]
 spine_02 [
  {shapeType=Capsule, length=3.5118, radius=8.7417, center=<-2.6885 -0.5337 0>}
  {shapeType=Capsule, length=19.1875, radius=10.8125, center=<-1.8317 1.9362 0>, rotation=<rot 90 0 0>}
 ]
 spine_03 [
  {shapeType=Capsule, length=6.5613, radius=9.1269, center=<0.0574 0.4416 0>}
  {shapeType=Capsule, length=21.097, radius=8.903, center=<-1.2163 0.549 0>, rotation=<rot 90 180 -175.0732>}
 ]
 spine_05 [{shapeType=Capsule, length=12.7001, radius=9.335, center=<0.6969 1.91 0>}]
 thigh_l []
 thigh_r []
 upperarm_l []
 upperarm_r []
}
constraints [
 {bone1Name=dyn_cape_02, bone2Name=spine_05, swing1LimitDegrees=20, swing2LimitDegrees=10, twistLimitDegrees=5}
 {bone1Name=dyn_cape_03, bone2Name=dyn_cape_02, swing1LimitDegrees=45, swing2LimitDegrees=20, twistLimitDegrees=5}
 {bone1Name=dyn_cape_04, bone2Name=dyn_cape_03, swing1LimitDegrees=45, swing2LimitDegrees=45, twistLimitDegrees=5}
 {
  bone1Name=dyn_cape_r_04
  bone2Name=dyn_cape_04
  swing1LimitDegrees=45
  swing1Motion=Free
  swing2LimitDegrees=45
  swing2Motion=Free
  twistLimitDegrees=45
  twistMotion=Free
 }
 {
  bone1Name=dyn_cape_l_04
  bone2Name=dyn_cape_04
  swing1LimitDegrees=45
  swing1Motion=Free
  swing2LimitDegrees=45
  swing2Motion=Free
  twistLimitDegrees=45
  twistMotion=Free
 }
 {bone1Name=dyn_cape_05, bone2Name=dyn_cape_04, swing1LimitDegrees=45, swing2LimitDegrees=45, twistLimitDegrees=5}
 {
  bone1Name=dyn_cape_r_05
  bone2Name=dyn_cape_05
  swing1LimitDegrees=45
  swing1Motion=Free
  swing2LimitDegrees=45
  swing2Motion=Free
  twistLimitDegrees=45
  twistMotion=Free
 }
 {
  bone1Name=dyn_cape_l_05
  bone2Name=dyn_cape_05
  swing1LimitDegrees=45
  swing1Motion=Free
  swing2LimitDegrees=45
  swing2Motion=Free
  twistLimitDegrees=45
  twistMotion=Free
 }
 {bone1Name=dyn_cape_06, bone2Name=dyn_cape_05, swing1LimitDegrees=45, swing2LimitDegrees=45, twistLimitDegrees=5}
 {
  bone1Name=dyn_cape_r_06
  bone2Name=dyn_cape_06
  swing1LimitDegrees=45
  swing1Motion=Free
  swing2LimitDegrees=45
  swing2Motion=Free
  twistLimitDegrees=45
  twistMotion=Free
 }
 {
  bone1Name=dyn_cape_l_06
  bone2Name=dyn_cape_06
  swing1LimitDegrees=45
  swing1Motion=Free
  swing2LimitDegrees=45
  swing2Motion=Free
  twistLimitDegrees=45
  twistMotion=Free
 }
 {
  bone1Name=dyn_cape_06
  bone2Name=dyn_cape_r_06
  swing1LimitDegrees=45
  swing1Motion=Free
  swing2LimitDegrees=45
  swing2Motion=Free
  twistLimitDegrees=45
  twistMotion=Free
 }
 {
  bone1Name=dyn_cape_06
  bone2Name=dyn_cape_l_06
  swing1LimitDegrees=45
  swing1Motion=Free
  swing2LimitDegrees=45
  swing2Motion=Free
  twistLimitDegrees=45
  twistMotion=Free
 }
 {bone1Name=dyn_cape_07, bone2Name=dyn_cape_06, swing1LimitDegrees=45, swing2LimitDegrees=45, twistLimitDegrees=5}
 {
  bone1Name=dyn_cape_l_07
  bone2Name=dyn_cape_07
  swing1LimitDegrees=45
  swing1Motion=Free
  swing2LimitDegrees=45
  swing2Motion=Free
  twistLimitDegrees=45
  twistMotion=Free
 }
 {
  bone1Name=dyn_cape_r_07
  bone2Name=dyn_cape_07
  swing1LimitDegrees=45
  swing1Motion=Free
  swing2LimitDegrees=45
  swing2Motion=Free
  twistLimitDegrees=45
  twistMotion=Free
 }
 {
  bone1Name=dyn_cape_07
  bone2Name=dyn_cape_r_07
  swing1LimitDegrees=45
  swing1Motion=Free
  swing2LimitDegrees=45
  swing2Motion=Free
  twistLimitDegrees=45
  twistMotion=Free
 }
 {
  bone1Name=dyn_cape_07
  bone2Name=dyn_cape_l_07
  swing1LimitDegrees=45
  swing1Motion=Free
  swing2LimitDegrees=45
  swing2Motion=Free
  twistLimitDegrees=45
  twistMotion=Free
 }
 {bone1Name=dyn_cape_l_02, bone2Name=spine_05, swing1LimitDegrees=20, swing2LimitDegrees=10, twistLimitDegrees=5}
 {bone1Name=dyn_cape_l_03, bone2Name=dyn_cape_l_02, swing1LimitDegrees=45, swing2LimitDegrees=20, twistLimitDegrees=5}
 {bone1Name=dyn_cape_l_04, bone2Name=dyn_cape_l_03, swing1LimitDegrees=45, swing2LimitDegrees=45, twistLimitDegrees=5}
 {bone1Name=dyn_cape_l_05, bone2Name=dyn_cape_l_04, swing1LimitDegrees=45, swing2LimitDegrees=45, twistLimitDegrees=5}
 {bone1Name=dyn_cape_l_06, bone2Name=dyn_cape_l_05, swing1LimitDegrees=45, swing2LimitDegrees=45, twistLimitDegrees=5}
 {bone1Name=dyn_cape_l_07, bone2Name=dyn_cape_l_06, swing1LimitDegrees=45, swing2LimitDegrees=45, twistLimitDegrees=5}
 {bone1Name=dyn_cape_r_02, bone2Name=spine_05, swing1LimitDegrees=20, swing2LimitDegrees=10, twistLimitDegrees=5}
 {bone1Name=dyn_cape_r_03, bone2Name=dyn_cape_r_02, swing1LimitDegrees=45, swing2LimitDegrees=20, twistLimitDegrees=5}
 {bone1Name=dyn_cape_r_04, bone2Name=dyn_cape_r_03, swing1LimitDegrees=45, swing2LimitDegrees=45, twistLimitDegrees=5}
 {bone1Name=dyn_cape_r_05, bone2Name=dyn_cape_r_04, swing1LimitDegrees=45, swing2LimitDegrees=45, twistLimitDegrees=5}
 {bone1Name=dyn_cape_r_06, bone2Name=dyn_cape_r_05, swing1LimitDegrees=45, swing2LimitDegrees=45, twistLimitDegrees=5}
 {bone1Name=dyn_cape_r_07, bone2Name=dyn_cape_r_06, swing1LimitDegrees=45, swing2LimitDegrees=45, twistLimitDegrees=5}
]
