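# Animation Blueprint for the face accessory skeletal mesh of the male medium ChiliAbide player cosmetic body.
# The AnimGraph outputs a Copy Pose From Mesh node set to use the attached parent, so the accessory takes its pose from the mesh it is attached to.
# The EventGraph stores the owning skeletal mesh component in Skeletal_Mesh on every animation update, and that variable feeds the node's source mesh input.
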AnimBlueprint /BRCosmetics/Characters/Player/Male/Medium/Bodies/M_MED_ChiliAbide/Meshes/Parts/M_MED_ChiliAbide_FaceAcc_AnimBP
target_skeleton=&/M_MED_ChiliAbide_FaceAcc_Skeleton
graphs [
 AnimGraph {nodes [Output Pose(Result=Copy Pose From Mesh(SourceMeshComponent=Get Skeletal_Mesh() -> (Skeletal_Mesh:Skeletal Mesh Component Object Reference)) {bUseAttachedParent=true}) {layerGroup=None}]}
 EventGraph {nodes [
   Event BlueprintUpdateAnimation()
   > Set Skeletal_Mesh(Skeletal_Mesh:Skeletal Mesh Component Object Reference=GetOwnerSkeletalMeshComponent()) -> (Output_Get:Skeletal Mesh Component Object Reference)
  ]}
]
variables [Skeletal_Mesh {type=SkeletalMeshComponent}]
